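# UI Library material function that builds X, Y and Z axis rotation matrices from a Float3 angle, for rotating a UV plane.
# It breaks the angle into components, takes sine, cosine and negated sine of each, and places them with constant ones and zeros as named-reroute matrix cells.
# MakeFloat3 packs the cells into nine row-vector outputs, three per axis, leaving the multiplication with the plane outside this function.
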MaterialFunction /FortUILibrary/MaterialFunctions/Effects/MF_Matrix_GenerateRotationMatrices
graph {
 properties {
  description="To rotate the UV plane, we need to perform a matrix rotation. We first generate a rotation matrices that will need to be multiplied by the original plane in order to perform the rotation."
  previewMaterialDomain=MD_UI
  libraryCategoriesText [UI Library]
 }
 nodes [
  n1 = NamedRerouteUsage()
  n2 = NamedRerouteUsage()
  n3 = NamedRerouteUsage()
  FunctionOutput(None=NamedRerouteDeclaration(None=MaterialFunctionCall(X=@n1, Y=@n2, Z=@n3) {materialFunction=&MakeFloat3}.Result) {name=Matrix X - Row 1, nodeColor=<rgba 0.9286 1 0 1>}) {description="Output values : X11, X12, X13", outputName="Matrix X - Row 1 (V3)", sortPriority=1}
  n4 = NamedRerouteUsage()
  n5 = NamedRerouteUsage()
  n6 = NamedRerouteUsage()
  FunctionOutput(None=NamedRerouteDeclaration(None=MaterialFunctionCall(X=@n4, Y=@n5, Z=@n6) {materialFunction=&MakeFloat3}.Result) {name=Matrix X - Row 2, nodeColor=<rgba 0.9286 1 0 1>}) {description="Output values : X21, X22, X23", outputName="Matrix X - Row 2 (V3)", sortPriority=2}
  n7 = NamedRerouteUsage()
  n8 = NamedRerouteUsage()
  n9 = NamedRerouteUsage()
  FunctionOutput(None=NamedRerouteDeclaration(None=MaterialFunctionCall(X=@n7, Y=@n8, Z=@n9) {materialFunction=&MakeFloat3}.Result) {name=Matrix Z - Row 2, nodeColor=<rgba 1 0 0.2719 1>}) {description="Output values : Z21, Z22, Z23", outputName="Matrix Z - Row 2 (V3)", sortPriority=8}
  n10 = NamedRerouteUsage()
  n11 = NamedRerouteUsage()
  n12 = NamedRerouteUsage()
  FunctionOutput(None=NamedRerouteDeclaration(None=MaterialFunctionCall(X=@n10, Y=@n11, Z=@n12) {materialFunction=&MakeFloat3}.Result) {name=Matrix Z - Row 3, nodeColor=<rgba 1 0 0.2719 1>}) {description="Output values : Z31, Z32, Z33", outputName="Matrix Z - Row 3 (V3)", sortPriority=9}
  n13 = NamedRerouteUsage()
  n14 = NamedRerouteUsage()
  n15 = NamedRerouteUsage()
  FunctionOutput(None=NamedRerouteDeclaration(None=MaterialFunctionCall(X=@n13, Y=@n14, Z=@n15) {materialFunction=&MakeFloat3}.Result) {name=Matrix X - Row 3, nodeColor=<rgba 0.9286 1 0 1>}) {description="Output values : X31, X32, X33", outputName="Matrix X - Row 3 (V3)", sortPriority=3}
  n16 = NamedRerouteUsage()
  n17 = NamedRerouteUsage()
  n18 = NamedRerouteUsage()
  FunctionOutput(None=NamedRerouteDeclaration(None=MaterialFunctionCall(X=@n16, Y=@n17, Z=@n18) {materialFunction=&MakeFloat3}.Result) {name=Matrix Y - Row 1, nodeColor=<rgba 0.2417 0.1295 1 1>}) {description="Output values : Y11, Y12, Y13", outputName="Matrix Y - Row 1 (V3)", sortPriority=4}
  n19 = NamedRerouteUsage()
  n20 = NamedRerouteUsage()
  n21 = NamedRerouteUsage()
  FunctionOutput(None=NamedRerouteDeclaration(None=MaterialFunctionCall(X=@n19, Y=@n20, Z=@n21) {materialFunction=&MakeFloat3}.Result) {name=Matrix Y - Row 2, nodeColor=<rgba 0.2417 0.1295 1 1>}) {description="Output values : Y21, Y22, Y23", outputName="Matrix Y - Row 2 (V3)", sortPriority=5}
  n22 = NamedRerouteUsage()
  n23 = NamedRerouteUsage()
  n24 = NamedRerouteUsage()
  FunctionOutput(None=NamedRerouteDeclaration(None=MaterialFunctionCall(X=@n22, Y=@n23, Z=@n24) {materialFunction=&MakeFloat3}.Result) {name=Matrix Y - Row 3, nodeColor=<rgba 0.2417 0.1295 1 1>}) {description="Output values : Y31, Y32, Y33", outputName="Matrix Y - Row 3 (V3)", sortPriority=6}
  n25 = NamedRerouteUsage()
  n26 = NamedRerouteUsage()
  n27 = NamedRerouteUsage()
  FunctionOutput(None=NamedRerouteDeclaration(None=MaterialFunctionCall(X=@n25, Y=@n26, Z=@n27) {materialFunction=&MakeFloat3}.Result) {name=Matrix Z - Row 1, nodeColor=<rgba 1 0 0.2719 1>}) {description="Output values : Z11, Z12, Z13", outputName="Matrix Z - Row 1 (V3)", sortPriority=7}
  NamedRerouteDeclaration(None=Constant() {r=1}) {name=X11, nodeColor=<rgba 0.9286 1 0 1>}
  NamedRerouteDeclaration(None=Constant()) {name=X12, nodeColor=<rgba 0.9286 1 0 1>}
  NamedRerouteDeclaration(None=Constant()) {name=X13, nodeColor=<rgba 0.9286 1 0 1>}
  NamedRerouteDeclaration(None=Constant()) {name=X21, nodeColor=<rgba 0.9286 1 0 1>}
  NamedRerouteDeclaration(None=Constant()) {name=X31, nodeColor=<rgba 0.9286 1 0 1>}
  n28 = MaterialFunctionCall(Float3=FunctionInput() {bUsePreviewValueAsDefault=true, description=Float 3 angle required, inputName=Angle, sortPriority=1}) {materialFunction=&BreakOutFloat3Components}
  NamedRerouteDeclaration(None=@n28.R) {name=Angle.x, nodeColor=<rgba 0.1016 1 0 1>}
  NamedRerouteDeclaration(None=@n28.G) {name=Angle.y, nodeColor=<rgba 0.1016 1 0 1>}
  NamedRerouteDeclaration(None=@n28.B) {name=Angle.z, nodeColor=<rgba 0.1016 1 0 1>}
  n29 = NamedRerouteUsage()
  NamedRerouteDeclaration(None=NamedRerouteDeclaration(None=Cosine(None=@n29)) {name="cos(angle.x)", nodeColor=<rgba 0 0.0807 1 1>}) {name=X22, nodeColor=<rgba 0.9286 1 0 1>}
  n30 = NamedRerouteUsage()
  NamedRerouteDeclaration(None=NamedRerouteDeclaration(None=Sine(None=@n30)) {name="sin(angle.x)", nodeColor=<rgba 0 0.0807 1 1>}) {name=X32, nodeColor=<rgba 0.9286 1 0 1>}
  n31 = NamedRerouteUsage()
  NamedRerouteDeclaration(None=NamedRerouteDeclaration(None=Multiply(A=@n31) {constB=-1}) {name="-sin(angle.x)", nodeColor=<rgba 0 0.0807 1 1>}) {name=X23, nodeColor=<rgba 0.9286 1 0 1>}
  NamedRerouteDeclaration(None=NamedRerouteUsage()) {name=X33, nodeColor=<rgba 0.9286 1 0 1>}
  n32 = NamedRerouteUsage()
  NamedRerouteDeclaration(None=NamedRerouteDeclaration(None=Cosine(None=@n32)) {name="cos(angle.y)", nodeColor=<rgba 0 0.0807 1 1>}) {name=Y11, nodeColor=<rgba 0.2417 0.1295 1 1>}
  NamedRerouteDeclaration(None=Constant()) {name=Y21, nodeColor=<rgba 0.2417 0.1295 1 1>}
  n33 = NamedRerouteUsage()
  NamedRerouteDeclaration(None=NamedRerouteDeclaration(None=Multiply(A=@n33) {constB=-1}) {name="-sin(angle.y)", nodeColor=<rgba 0 0.0807 1 1>}) {name=Y31, nodeColor=<rgba 0.2417 0.1295 1 1>}
  NamedRerouteDeclaration(None=Constant()) {name=Y12, nodeColor=<rgba 0.2417 0.1295 1 1>}
  NamedRerouteDeclaration(None=Constant() {r=1}) {name=Y22, nodeColor=<rgba 0.2417 0.1295 1 1>}
  NamedRerouteDeclaration(None=Constant()) {name=Y32, nodeColor=<rgba 0.2417 0.1295 1 1>}
  n34 = NamedRerouteUsage()
  NamedRerouteDeclaration(None=NamedRerouteDeclaration(None=Sine(None=@n34)) {name="sin(angle.y)", nodeColor=<rgba 0 0.0807 1 1>}) {name=Y13, nodeColor=<rgba 0.2417 0.1295 1 1>}
  NamedRerouteDeclaration(None=Constant()) {name=Y23, nodeColor=<rgba 0.2417 0.1295 1 1>}
  NamedRerouteDeclaration(None=NamedRerouteUsage()) {name=Y33, nodeColor=<rgba 0.2417 0.1295 1 1>}
  n35 = NamedRerouteUsage()
  NamedRerouteDeclaration(None=NamedRerouteDeclaration(None=Cosine(None=@n35)) {name="cos(angle.z)", nodeColor=<rgba 0 0.0807 1 1>}) {name=Z11, nodeColor=<rgba 1 0 0.2719 1>}
  n36 = NamedRerouteUsage()
  NamedRerouteDeclaration(None=NamedRerouteDeclaration(None=Sine(None=@n36)) {name="sin(angle.z)", nodeColor=<rgba 0 0.0807 1 1>}) {name=Z21, nodeColor=<rgba 1 0 0.2719 1>}
  NamedRerouteDeclaration(None=Constant()) {name=Z31, nodeColor=<rgba 1 0 0.2719 1>}
  n37 = NamedRerouteUsage()
  NamedRerouteDeclaration(None=NamedRerouteDeclaration(None=Multiply(A=@n37) {constB=-1}) {name="-sin(angle.z)", nodeColor=<rgba 0 0.0807 1 1>}) {name=Z12, nodeColor=<rgba 1 0 0.2719 1>}
  NamedRerouteDeclaration(None=NamedRerouteUsage()) {name=Z22, nodeColor=<rgba 1 0 0.2719 1>}
  NamedRerouteDeclaration(None=Constant()) {name=Z32, nodeColor=<rgba 1 0 0.2719 1>}
  NamedRerouteDeclaration(None=Constant()) {name=Z13, nodeColor=<rgba 1 0 0.2719 1>}
  NamedRerouteDeclaration(None=Constant()) {name=Z23, nodeColor=<rgba 1 0 0.2719 1>}
  NamedRerouteDeclaration(None=Constant() {r=1}) {name=Z33, nodeColor=<rgba 1 0 0.2719 1>}
 ]
}
