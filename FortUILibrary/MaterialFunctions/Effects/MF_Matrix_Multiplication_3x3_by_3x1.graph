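# Material function in the UI library that multiplies a 3x3 matrix by a 3x1 matrix.
# Matrix A arrives as three row vectors and Matrix B as one column vector, which a named reroute shares across three dot products.
# MakeFloat3 packs the three results into one vector output holding C11, C21 and C31, a column despite its "Row 1" label.
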MaterialFunction /FortUILibrary/MaterialFunctions/Effects/MF_Matrix_Multiplication_3x3_by_3x1
graph {
 properties {description=Matrix Multiplication 3x3 by 3x1, libraryCategoriesText [UI Library]}
 nodes [
  n1 = FunctionInput() {bUsePreviewValueAsDefault=true, description="Expected values: A11, A12, A13", inputName=Matrix A - Row 1, sortPriority=1}
  n2 = NamedRerouteUsage()
  n3 = FunctionInput() {bUsePreviewValueAsDefault=true, description="Expected values: A21, A22, A23", inputName=Matrix A - Row 2, sortPriority=2}
  n4 = NamedRerouteUsage()
  n5 = FunctionInput() {bUsePreviewValueAsDefault=true, description="Expected values: A31, A32, A33", inputName=Matrix A - Row 3, sortPriority=3}
  n6 = NamedRerouteUsage()
  FunctionOutput(None=MaterialFunctionCall(X=DotProduct(A=@n1, B=@n2), Y=DotProduct(A=@n3, B=@n4), Z=DotProduct(A=@n5, B=@n6)) {materialFunction=&MakeFloat3}.Result) {description="Output values : C11, C21, C31", outputName="Matrix C - Row 1 (V3)", sortPriority=1}
  NamedRerouteDeclaration(None=FunctionInput() {bUsePreviewValueAsDefault=true, description="Expected values: B11, B31, B31", inputName=Matrix B - Column 1, sortPriority=4}) {name=Matrix B Column 1, nodeColor=<rgba 0.1759 0.066 1 1>}
 ]
}
