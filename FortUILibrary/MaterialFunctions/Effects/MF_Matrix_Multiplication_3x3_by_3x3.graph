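# Material function in the UI Library that multiplies two 3x3 matrices, A and B, each supplied as three float3 row inputs.
# It breaks B's rows into components and regroups them into three column vectors, shared with A's rows through named reroutes.
# Dot products between those vectors form the elements of C, which MakeFloat3 packs into three float3 row outputs.
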MaterialFunction /FortUILibrary/MaterialFunctions/Effects/MF_Matrix_Multiplication_3x3_by_3x3
graph {
 properties {description=Matrix Multiplication 3x3 by 3x3, libraryCategoriesText [UI Library]}
 nodes [
  n1 = NamedRerouteUsage()
  n2 = NamedRerouteUsage()
  n3 = NamedRerouteUsage()
  n4 = NamedRerouteUsage()
  n5 = NamedRerouteUsage()
  n6 = NamedRerouteUsage()
  FunctionOutput(None=MaterialFunctionCall(X=DotProduct(A=@n1, B=@n2), Y=DotProduct(A=@n3, B=@n4), Z=DotProduct(A=@n5, B=@n6)) {materialFunction=&MakeFloat3}.Result) {description="Output values : C11, C12, C13", outputName="Matrix C - Row 1 (V3)", sortPriority=1}
  n7 = NamedRerouteUsage()
  n8 = NamedRerouteUsage()
  n9 = NamedRerouteUsage()
  n10 = NamedRerouteUsage()
  n11 = NamedRerouteUsage()
  n12 = NamedRerouteUsage()
  FunctionOutput(None=MaterialFunctionCall(X=DotProduct(A=@n7, B=@n8), Y=DotProduct(A=@n9, B=@n10), Z=DotProduct(A=@n11, B=@n12)) {materialFunction=&MakeFloat3}.Result) {description="Output values : C21, C22, C23", outputName="Matrix C - Row 2 (V3)", sortPriority=2}
  n13 = NamedRerouteUsage()
  n14 = NamedRerouteUsage()
  n15 = NamedRerouteUsage()
  n16 = NamedRerouteUsage()
  n17 = NamedRerouteUsage()
  n18 = NamedRerouteUsage()
  FunctionOutput(None=MaterialFunctionCall(X=DotProduct(A=@n13, B=@n14), Y=DotProduct(A=@n15, B=@n16), Z=DotProduct(A=@n17, B=@n18)) {materialFunction=&MakeFloat3}.Result) {description="Output values : C31, C32, C33", outputName="Matrix C - Row 3 (V3)", sortPriority=3}
  n19 = MaterialFunctionCall(Float3=FunctionInput() {bUsePreviewValueAsDefault=true, description="Expected values: B11, B12, B13", inputName=Matrix B - Row 1, sortPriority=4}) {materialFunction=&BreakOutFloat3Components}
  n20 = MaterialFunctionCall(Float3=FunctionInput() {bUsePreviewValueAsDefault=true, description="Expected values: B21, B22, B23", inputName=Matrix B - Row 2, sortPriority=5}) {materialFunction=&BreakOutFloat3Components}
  n21 = MaterialFunctionCall(Float3=FunctionInput() {bUsePreviewValueAsDefault=true, description="Expected values: B31, B32, B33", inputName=Matrix B - Row 3, sortPriority=6}) {materialFunction=&BreakOutFloat3Components}
  NamedRerouteDeclaration(None=MaterialFunctionCall(X=@n19.G, Y=@n20.G, Z=@n21.G) {desc=G, materialFunction=&MakeFloat3}.Result) {name=Matrix B Column 2, nodeColor=<rgba 0.1759 0.066 1 1>}
  NamedRerouteDeclaration(None=MaterialFunctionCall(X=@n19.R, Y=@n20.R, Z=@n21.R) {desc=R, materialFunction=&MakeFloat3}.Result) {name=Matrix B Column 1, nodeColor=<rgba 0.1759 0.066 1 1>}
  NamedRerouteDeclaration(None=MaterialFunctionCall(X=@n19.B, Y=@n20.B, Z=@n21.B) {desc=B, materialFunction=&MakeFloat3}.Result) {name=Matrix B Column 3, nodeColor=<rgba 0.1759 0.066 1 1>}
  NamedRerouteDeclaration(None=FunctionInput() {bUsePreviewValueAsDefault=true, description="Expected values: A11, A12, A13", inputName=Matrix A - Row 1, sortPriority=1}) {name=Matrix A Row 1, nodeColor=<rgba 0.9286 1 0 1>}
  NamedRerouteDeclaration(None=FunctionInput() {bUsePreviewValueAsDefault=true, description="Expected values: A31, A32, A33", inputName=Matrix A - Row 3, sortPriority=3}) {name=Matrix A Row 3, nodeColor=<rgba 1 0.0158 0 1>}
  NamedRerouteDeclaration(None=FunctionInput() {bUsePreviewValueAsDefault=true, description="Expected values: A21, A22, A23", inputName=Matrix A - Row 2, sortPriority=2}) {name=Matrix A Row 2, nodeColor=<rgba 1 0.4268 0 1>}
 ]
}
