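# Material function in the UI Library that transposes a 3x3 matrix supplied as three float3 row inputs.
# Each input row is broken into components, stored in named reroutes A11 through A33.
# MakeFloat3 calls rebuild them from reroute usages into three outputs, each described as holding one column of the original matrix.
# Reroute colors distinguish the diagonal from the upper and lower triangles.
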MaterialFunction /FortUILibrary/MaterialFunctions/Effects/MF_Matrix_Transpose_3x3
graph {
 properties {description=Matrix Transpose 3x3, libraryCategoriesText [UI Library]}
 nodes [
  FunctionOutput(None=MaterialFunctionCall(X=NamedRerouteUsage(), Y=NamedRerouteUsage(), Z=NamedRerouteUsage()) {materialFunction=&MakeFloat3}.Result) {description="Output values : A11, A21, A31", outputName="Matrix A.T - Row 1 (V3)", sortPriority=1}
  FunctionOutput(None=MaterialFunctionCall(X=NamedRerouteUsage(), Y=NamedRerouteUsage(), Z=NamedRerouteUsage()) {materialFunction=&MakeFloat3}.Result) {description="Output values : A12, A22, A32", outputName="Matrix A.T - Row 2 (V3)", sortPriority=2}
  FunctionOutput(None=MaterialFunctionCall(X=NamedRerouteUsage(), Y=NamedRerouteUsage(), Z=NamedRerouteUsage()) {materialFunction=&MakeFloat3}.Result) {description="Output values : A13, A23, A33", outputName="Matrix A.T - Row 3 (V3)", sortPriority=3}
  n1 = MaterialFunctionCall(Float3=FunctionInput() {bUsePreviewValueAsDefault=true, description="Expected values: A11, A12, A13", inputName=Matrix A - Row 1, sortPriority=1}) {materialFunction=&BreakOutFloat3Components}
  NamedRerouteDeclaration(None=@n1.R) {name=A11, nodeColor=<rgba 0.035 0.035 0.035 1>}
  NamedRerouteDeclaration(None=@n1.G) {name=A12, nodeColor=<rgba 0 0.5968 1 1>}
  NamedRerouteDeclaration(None=@n1.B) {name=A13, nodeColor=<rgba 0 0.5968 1 1>}
  n2 = MaterialFunctionCall(Float3=FunctionInput() {bUsePreviewValueAsDefault=true, description="Expected values: A21, A22, A23", inputName=Matrix A - Row 2, sortPriority=2}) {materialFunction=&BreakOutFloat3Components}
  NamedRerouteDeclaration(None=@n2.R) {name=A21, nodeColor=<rgba 0.9286 1 0 1>}
  n3 = MaterialFunctionCall(Float3=FunctionInput() {bUsePreviewValueAsDefault=true, description="Expected values: A31, A32, A33", inputName=Matrix A - Row 3, sortPriority=3}) {materialFunction=&BreakOutFloat3Components}
  NamedRerouteDeclaration(None=@n3.R) {name=A31, nodeColor=<rgba 0.9286 1 0 1>}
  NamedRerouteDeclaration(None=@n2.G) {name=A22, nodeColor=<rgba 0.035 0.035 0.035 1>}
  NamedRerouteDeclaration(None=@n3.G) {name=A32, nodeColor=<rgba 0.9286 1 0 1>}
  NamedRerouteDeclaration(None=@n2.B) {name=A23, nodeColor=<rgba 0 0.5968 1 1>}
  NamedRerouteDeclaration(None=@n3.B) {name=A33, nodeColor=<rgba 0.035 0.035 0.035 1>}
 ]
}
