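# UI material function that renders animated, tinted procedural clouds, outputting Clouds, Alpha and FoggyAlpha.
# Two grayscale samples of a noise texture pan at different speeds over scaled UVs, with UI pixel size providing aspect correction.
# Math driven by mask, density and intensity inputs shapes the noise, and chained lerps add low, high and dark colors for fading and scattering.
# FoggyAlpha caps a floored fog value at the cloud alpha.
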MaterialFunction /FortUILibrary/MaterialFunctions/Effects/MF_UI_Clouds
graph {
 properties {bExposeToLibrary=true, previewBlendMode=BLEND_Translucent}
 nodes [
  n1 = NamedRerouteUsage()
  FunctionOutput(None=@n1) {outputName=Clouds}
  n2 = ComponentMask(None=@n1) {a=true}
  FunctionOutput(None=@n2) {outputName=Alpha}
  n3 = NamedRerouteUsage()
  n4 = NamedRerouteUsage()
  FunctionOutput(None=Min(A=@n2, B=Max(A=@n3, B=@n4) {constB=0.2})) {outputName=FoggyAlpha}
  NamedRerouteDeclaration(None=ComponentMask(None=TextureCoordinate()) {g=true}) {name=Mask - G, nodeColor=<rgba 0.4062 1 0 1>}
  n5 = FunctionInput(Preview=TextureObject() {samplerType=SAMPLERTYPE_Grayscale, texture=/FortUILibrary/MaterialFunctions/ExampleContent/T_Cloud_R}) {
   bUsePreviewValueAsDefault=true
   description=Use a 'TextureObject' as parameter
   inputName=NoiseTexture
   inputType=FunctionInput_Texture2D
   sortPriority=10
  }
  NamedRerouteDeclaration(None=@n5) {name=NoiseTexture, nodeColor=<rgba 1 0.5374 0.8137 1>}
  NamedRerouteDeclaration(None=FunctionInput(Preview=Constant() {r=1}) {bUsePreviewValueAsDefault=true, inputName=Mask, inputType=FunctionInput_Vector2, sortPriority=-1}) {name=Mask, nodeColor=<rgba 1 0 0.0297 1>}
  n6 = FunctionInput(Preview=Constant() {r=0.2}) {bUsePreviewValueAsDefault=true, description=Default: 0.2, inputName=FoggyAlphaMin, inputType=FunctionInput_Scalar, sortPriority=9}
  NamedRerouteDeclaration(None=@n6) {name=FoggyAlphaMin, nodeColor=<rgba 0.6406 0 1 1>}
  n7 = LinearInterpolate(B=OneMinus(None=NamedRerouteUsage()), Alpha=Saturate(None=NamedRerouteUsage()))
  n8 = NamedRerouteUsage()
  n9 = NamedRerouteUsage()
  n10 = NamedRerouteUsage()
  n11 = OneMinus(None=Power(Base=OneMinus(None=@n10)))
  n12 = Min(A=NamedRerouteUsage(), B=Divide(A=Add(A=@n8, B=@n9), B=Multiply(A=@n11) {constB=1.75}) {constB=1.75})
  n13 = NamedRerouteUsage()
  n14 = Saturate(None=Abs(None=Subtract(A=@n12, B=@n13)))
  n15 = Subtract(A=Saturate(None=Multiply(A=@n7, B=@n14)), B=@n14)
  NamedRerouteDeclaration(None=Saturate(None=Abs(None=@n15))) {name=CloudShaped, nodeColor=<rgba 1 0.7969 0 1>}
  n16 = NamedRerouteUsage()
  n17 = Constant()
  n18 = NamedRerouteUsage()
  n19 = Constant() {r=1}
  n20 = Multiply(A=NamedRerouteUsage()) {constB=20}
  NamedRerouteDeclaration(None=LinearInterpolate(A=AppendVector(A=@n16, B=@n17), B=AppendVector(A=@n18, B=@n19), Alpha=Saturate(None=@n20))) {name=Clouds+Alpha, nodeColor=<rgba 0 1 0.7422 1>}
  n21 = FunctionInput(Preview=Constant3Vector() {constant=<rgba 0.9387 0.8654 0.8758 0>}) {bUsePreviewValueAsDefault=true, inputName=CloudColorLow, sortPriority=1, previewValue=<0.7758 0.2462 0.4342 0>}
  NamedRerouteDeclaration(None=@n21) {name=ColorLow, nodeColor=<rgba 0.1406 0.1406 0.1406 1>}
  n22 = FunctionInput(Preview=Constant() {r=1}) {
   bUsePreviewValueAsDefault=true
   description="Quantity of clouds in the shape. | 0 -> 1 | Default: 1.0"
   inputName=CloudIntensity
   inputType=FunctionInput_Scalar
   previewValue=<0 0 0 0>
  }
  NamedRerouteDeclaration(None=@n22) {name=CloudIntensity, nodeColor=<rgba 0 1 0.4844 1>}
  n23 = FunctionInput(Preview=Constant3Vector() {constant=<rgba 0.7835 0.2502 0.4452 1>}) {bUsePreviewValueAsDefault=true, inputName=CloudColorHigh, sortPriority=3, previewValue=<0.7758 0.2462 0.4342 0>}
  NamedRerouteDeclaration(None=@n23) {name=ColorHigh, nodeColor=<rgba 0.0625 0.0625 0.0625 1>}
  n24 = FunctionInput(Preview=Constant3Vector() {constant=<rgba 0.2375 0.0762 0.1352 1>}) {bUsePreviewValueAsDefault=true, inputName=CloudColorDark, sortPriority=5, previewValue=<0.7758 0.2462 0.4342 0>}
  NamedRerouteDeclaration(None=@n24) {name=ColorDark, nodeColor=<rgba 0 0 0 1>}
  n25 = NamedRerouteUsage()
  n26 = Subtract(A=Multiply(A=NamedRerouteUsage(), B=@n25) {constB=1.5}, B=Divide(A=@n25)) {constB=0.7}
  NamedRerouteDeclaration(None=LinearInterpolate(A=NamedRerouteUsage(), B=NamedRerouteUsage(), Alpha=Saturate(None=@n26))) {name=Clouds - Fading, nodeColor=<rgba 0.2656 0 1 1>}
  n27 = NamedRerouteUsage()
  n28 = OneMinus(None=Saturate(None=Multiply(A=@n27) {constB=2}))
  n29 = OneMinus(None=Multiply(A=@n28, B=@n28))
  NamedRerouteDeclaration(None=LinearInterpolate(A=NamedRerouteUsage(), B=NamedRerouteUsage(), Alpha=@n29)) {name=Clouds - Back - Colored, nodeColor=<rgba 0 0.2266 1 1>}
  n30 = NamedRerouteUsage()
  n31 = Saturate(None=Subtract(A=NamedRerouteUsage(), B=Multiply(A=@n30) {constB=0.3}))
  n32 = Multiply(A=Multiply(A=@n31, B=@n31) {constB=1.5})
  NamedRerouteDeclaration(None=LinearInterpolate(A=NamedRerouteUsage(), B=NamedRerouteUsage(), Alpha=Saturate(None=@n32))) {name=Clouds + Scattering, nodeColor=<rgba 0 0.0625 1 1>}
  n33 = FunctionInput(Preview=Constant() {r=0.5}) {
   bUsePreviewValueAsDefault=true
   description="Num of times the it should repeat based on noise. | 0 -> Infinity | Default: .5"
   inputName=Scale
   inputType=FunctionInput_Scalar
   previewValue=<0.5 0 0 0>
  }
  NamedRerouteDeclaration(None=@n33) {name=Scale, nodeColor=<rgba 0.0312 0 1 1>}
  n34 = Multiply(A=NamedRerouteUsage(), B=NamedRerouteUsage()) {constB=0.5}
  NamedRerouteDeclaration(None=TextureSample(UVs=Panner(Coordinate=@n34) {speedX=-0.01, speedY=0.001}, Tex=NamedRerouteUsage()) {samplerType=SAMPLERTYPE_Grayscale}.R) {name=Panner - R, nodeColor=<rgba 1 0 0.6562 1>}
  n35 = FunctionInput(Preview=Constant() {r=1}) {
   bUsePreviewValueAsDefault=true
   description="0 -> 1 | Default: 1.0"
   inputName=Density
   inputType=FunctionInput_Scalar
   previewValue=<0 0 0 0>
  }
  NamedRerouteDeclaration(None=@n35) {name=Density, nodeColor=<rgba 0.1016 0 1 1>}
  n36 = Multiply(A=NamedRerouteUsage(), B=NamedRerouteUsage()) {constB=0.5}
  n37 = NamedRerouteUsage()
  NamedRerouteDeclaration(None=TextureSample(UVs=Panner(Coordinate=@n36, Speed=@n37) {speedX=0.005}, Tex=NamedRerouteUsage()) {samplerType=SAMPLERTYPE_Grayscale}.R) {name=Panner + R, nodeColor=<rgba 1 0 0.6562 1>}
  n38 = MaterialFunctionCall() {materialFunction=&GetUserInterfaceUV}
  n39 = Divide(A=ComponentMask(None=@n38.Pixel Size) {r=true}, B=ComponentMask(None=@n38.Pixel Size) {g=true})
  n40 = TextureCoordinate()
  n41 = ComponentMask(None=@n40) {r=true}
  NamedRerouteDeclaration(None=AppendVector(A=Multiply(A=@n39, B=@n41), B=ComponentMask(None=@n40) {g=true})) {name=Aspect Ratio, nodeColor=<rgba 0 1 0.6953 1>}
  n42 = FunctionInput(Preview=Constant() {r=1.5}) {
   bUsePreviewValueAsDefault=true
   description=Default: 1.5
   inputName=CloudDarkIntensity
   inputType=FunctionInput_Scalar
   sortPriority=6
   previewValue=<0 0 0 0>
  }
  NamedRerouteDeclaration(None=@n42) {name=Dark Intensity, nodeColor=<rgba 0.8984 1 0 1>}
  n43 = Constant2Vector() {r=-5}
  NamedRerouteDeclaration(None=Divide(A=FunctionInput(Preview=@n43) {bUsePreviewValueAsDefault=true, description="Default: (-5.0, 0.0)", inputName=Speed, inputType=FunctionInput_Vector2}) {constB=1000}) {name=Speed, nodeColor=<rgba 0.5234 1 0 1>}
  NamedRerouteDeclaration(None=@n29) {name=FoggyAlpha, nodeColor=<rgba 0.5469 0 1 1>}
  Reroute()
  SmoothStep(Value=ComponentMask(None=TextureCoordinate()) {g=true}) {constMax=0.2, constMin=0.05, constValue=0.4}
 ]
}
