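# UI material function that distorts a texture along a flow map to give looping flowing motion, returning RGB.
# Flow map red-green becomes signed vectors, scaled by intensity with X flipped, that offset the texture UVs.
# Time times speed, phase-shifted by flow map red, drives two half-cycle-apart samples, crossfaded with a triangle wave to hide resets.
# A leftover channel-minimum branch is unconnected.
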MaterialFunction /FortUILibrary/MaterialFunctions/Effects/MF_UI_FlowMap
graph {
 properties {bExposeToLibrary=true, description=Simple flow map function, libraryCategoriesText [UI Shader Library]}
 nodes [
  n1 = FunctionInput(Preview=Constant() {r=0.25}) {
   bUsePreviewValueAsDefault=true
   description="Determins how intense the flow distortion is (.25 default)"
   inputName=FlowIntensity
   inputType=FunctionInput_Scalar
   sortPriority=5
  }
  n2 = FunctionInput(Preview=TextureCoordinate()) {bUsePreviewValueAsDefault=true, description=Flow map UVs, inputName=FlowMap UVs, inputType=FunctionInput_Vector2, sortPriority=4}
  n3 = FunctionInput(Preview=TextureObject() {samplerType=SAMPLERTYPE_LinearColor, texture=/Game/UI/Frontend/ItemShop/Art/WaterBackground/T_UI_WaterFlowMap}) {
   bUsePreviewValueAsDefault=true
   description=This is the flow map texture you want to distort the UVs
   inputName=FlowMap Texture 2D
   inputType=FunctionInput_Texture2D
   sortPriority=3
  }
  n4 = TextureSample(UVs=@n2, Tex=@n3)
  n5 = ConstantBiasScale(None=@n4.RGB) {bias=-0.5, scale=2}
  n6 = Multiply(A=Constant2Vector() {g=1, r=-1}, B=Multiply(A=@n1, B=ComponentMask(None=@n5) {g=true, r=true}))
  n7 = FunctionInput(Preview=Constant() {r=0.5}) {
   bUsePreviewValueAsDefault=true
   description="This determines how fast the flow map texture loops (.5 default)"
   inputName=FlowSpeed
   inputType=FunctionInput_Scalar
   sortPriority=6
  }
  n8 = Add(A=Multiply(A=Time(), B=@n7), B=Max(A=@n4.R, B=@n4.R) {desc=For Tori and Adri to test more})
  n9 = Frac(None=@n8)
  n10 = TextureCoordinate()
  n11 = FunctionInput(Preview=@n10) {
   bUsePreviewValueAsDefault=true
   description=Texture UVs on the flow map
   inputName=Texture UVs
   inputType=FunctionInput_Vector2
   sortPriority=1
  }
  n12 = Add(A=Multiply(A=NamedRerouteDeclaration(None=@n6) {name=FlowMapCoords, nodeColor=<rgba 0 0 0 1>}, B=@n9), B=@n11)
  n13 = NamedRerouteUsage()
  n14 = Add(A=@n8) {constB=0.5}
  n15 = Add(A=@n11, B=Multiply(A=NamedRerouteUsage(), B=Frac(None=@n14)))
  n16 = NamedRerouteUsage()
  n17 = ConstantBiasScale(None=@n9) {bias=-0.5, scale=2}
  FunctionOutput(None=LinearInterpolate(A=TextureSample(UVs=@n12, Tex=@n13).RGB, B=TextureSample(UVs=@n15, Tex=@n16).RGB, Alpha=Abs(None=@n17)))
  n18 = FunctionInput(Preview=TextureObject()) {
   bUsePreviewValueAsDefault=true
   description=This is the texture you want to apply the flow distortion to
   inputName=Texture 2D
   inputType=FunctionInput_Texture2D
  }
  n19 = TextureSample(UVs=@n10, Tex=NamedRerouteDeclaration(None=@n18) {name=TextureNode, nodeColor=<rgba 0 0 0 1>})
  Min(A=Min(A=@n19.R, B=@n19.R), B=@n19.B)
 ]
}
