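# UI material function that warps UVs into a funnel converging on a center point, outputting raw and wrapped UVs plus a center hole mask.
# MF_UI_CheaperVectorToRadial supplies angle and radius, with Tiling and Offset controlling circumference and distance.
# Blend Distance lerps the depth axis between a divide-based tunnel and a multiply-based flattened funnel, then Fmod bounds it for UV cell randomization.
# Mask Size remaps the doubled radius into the linear hole mask.
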MaterialFunction /FortUILibrary/MaterialFunctions/Effects/MF_UI_Funnel
graph {
 properties {
  bExposeToLibrary=true
  description=Create a funnel effect by adjusting the UVs to converge toward a center point
  previewMaterialDomain=MD_UI
 }
 nodes [
  n1 = FunctionInput(Preview=MaterialFunctionCall() {materialFunction=&GetUserInterfaceUV}.Normalized UV) {bUsePreviewValueAsDefault=true, description=Set the UV coordinates, inputName=UV, inputType=FunctionInput_Vector2}
  n2 = NamedRerouteDeclaration(None=MaterialFunctionCall!(Vector or UVs=NamedRerouteDeclaration(None=@n1) {name=UV, nodeColor=<rgba 0 0 0 1>}) {materialFunction=/FortUILibrary/MaterialFunctions/Basics/MF_UI_CheaperVectorToRadial}.Radial Coordinates) {name=Radial Coordinates, nodeColor=<rgba 0 0 0 1>}
  n3 = NamedRerouteDeclaration(None=Multiply(A=ComponentMask(None=@n2) {g=true}) {constB=2}) {name=Hole Mask, nodeColor=<rgba 0 0 0 1>}
  n4 = FunctionInput() {
   bUsePreviewValueAsDefault=true
   description=Set the hole mask radius
   inputName=Mask Size
   inputType=FunctionInput_Scalar
   sortPriority=20
   previewValue=<0.25 1 0 1>
  }
  FunctionOutput(None=MaterialFunctionCall(Input=Saturate(None=@n3), Input Low=@n4, Input High=Constant() {r=1}) {materialFunction=&RemapValue_0-1}.Result) {description=Linear mask of the center of the funnel, outputName=Hole Mask, sortPriority=10}
  n5 = NamedRerouteUsage()
  n6 = NamedRerouteUsage()
  n7 = FunctionInput() {
   bUsePreviewValueAsDefault=true
   description=Blend between a tunnel effect or a flatten funnel
   inputName=Blend Distance
   inputType=FunctionInput_Scalar
   sortPriority=40
   previewValue=<0 0 0 0>
  }
  n8 = LinearInterpolate(A=Divide(A=NamedRerouteUsage(), B=Max(A=@n5) {constB=0}), B=Multiply(A=NamedRerouteUsage(), B=Multiply(A=@n6) {constB=-1}), Alpha=@n7)
  n9 = Multiply(A=ComponentMask(None=NamedRerouteUsage()) {g=true}) {constB=-1}
  n10 = AppendVector(A=NamedRerouteUsage(), B=Fmod(A=Add(A=@n8, B=@n9), B=Constant() {r=8}) {desc=Fix issue to allow randomization of UV cells})
  FunctionOutput(None=Frac(None=FunctionOutput(None=@n10) {description=Output the UVs mapped to a funnel effect, outputName=Funnel UV, sortPriority=5})) {description=Output the fractional UVs mapped to a funnel effect, outputName=Wrapped Funnel UV}
  n11 = FunctionInput() {
   bUsePreviewValueAsDefault=true
   description="Set the tiling amount (X is Tiling over Circumference, Y is Tiling over Distance)"
   inputName=Tiling
   inputType=FunctionInput_Vector2
   sortPriority=5
   previewValue=<8 2 0 1>
  }
  n12 = NamedRerouteDeclaration(None=@n11) {name=Tiling Value, nodeColor=<rgba 0 0 0 1>}
  NamedRerouteDeclaration(None=ComponentMask(None=@n12) {desc="Tiling over Circumference ", r=true}) {name=Circumference Tiling, nodeColor=<rgba 0 0 0 1>}
  NamedRerouteDeclaration(None=ComponentMask(None=@n12) {desc=Tiling Over Distance, g=true}) {name=Distance Tiling, nodeColor=<rgba 0 0 0 1>}
  n13 = Add(A=ComponentMask(None=NamedRerouteUsage()) {r=true}, B=ComponentMask(None=@n2) {r=true})
  NamedRerouteDeclaration(None=Multiply(A=Frac(None=@n13), B=NamedRerouteUsage())) {name=Tunnel Circumference, nodeColor=<rgba 0 0 0 1>}
  n14 = FunctionInput() {
   bUsePreviewValueAsDefault=true
   description="Set the offset (X is offset around circumference, Y is offset toward center)"
   inputName=Offset
   inputType=FunctionInput_Vector2
   sortPriority=10
  }
  NamedRerouteDeclaration(None=@n14) {name=Offset Amount, nodeColor=<rgba 0 0 0 1>}
 ]
}
