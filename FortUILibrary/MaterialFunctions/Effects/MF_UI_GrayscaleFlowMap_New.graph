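# UI material function that animates a texture with a simplified flow-map distortion driven by a grayscale map.
# A scalar direction becomes a 2D vector via saturated linear ramps, blended from neutral by map times intensity.
# Time, speed and a per-pixel offset drive two half-cycle-shifted UV samples, crossfaded by a triangle wave to hide resets.
# It outputs the distorted texture, both UV sets and the blend mask.
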MaterialFunction /FortUILibrary/MaterialFunctions/Effects/MF_UI_GrayscaleFlowMap_New
graph {
 properties {
  bExposeToLibrary=true
  description=Simplified flow map effect that allows to distort by a grayscale directionality map
  previewMaterialDomain=MD_UI
  libraryCategoriesText [UI Shader Library]
 }
 nodes [
  n1 = NamedRerouteUsage()
  n2 = NamedRerouteUsage()
  n3 = FunctionInput() {
   bUsePreviewValueAsDefault=true
   description=The speed of the flow map
   inputName=Flow Speed
   inputType=FunctionInput_Scalar
   sortPriority=6
   previewValue=<0.25 0 0 1>
  }
  n4 = Add(A=Max(A=ComponentMask(None=@n2) {g=true}, B=ComponentMask(None=@n2) {r=true}), B=Multiply(A=@n3, B=Time()) {constB=4})
  n5 = Frac(None=@n4)
  n6 = FunctionInput(Preview=TextureCoordinate()) {bUsePreviewValueAsDefault=true, description=The UVs the flow map applies to, inputName=UVs, inputType=FunctionInput_Vector2}
  n7 = FunctionOutput(None=Add(A=Multiply(A=@n1, B=@n5) {constB=4}, B=@n6)) {description=Output the initial distorted UV, outputName=UV, sortPriority=2}
  n8 = FunctionInput(Preview=TextureObject()) {description=The texture the flow map applies to, inputName=Texture Object, inputType=FunctionInput_Texture2D, sortPriority=1}
  n9 = NamedRerouteUsage()
  n10 = Frac(None=Add(A=@n4) {constB=0.5})
  n11 = FunctionOutput(None=Add(A=@n6, B=Multiply(A=@n9, B=@n10) {constB=4})) {description=Output the offsetted distorted UV, outputName=Offset UV, sortPriority=3}
  n12 = Abs(None=ConstantBiasScale(None=@n5) {bias=-0.5, scale=2})
  FunctionOutput(None=LinearInterpolate(
    A=TextureSample(UVs=@n7, Tex=@n8).RGBA
    B=TextureSample(UVs=@n11, Tex=@n8).RGBA
    Alpha=FunctionOutput(None=@n12) {description=Output a mask of the distortion applied, outputName=Distortion Mask, sortPriority=5}
   )) {description=Output the result of the distortion applied to the initial texture, outputName=Distorted Texture}
  n13 = Constant2Vector() {desc=Not Moving Flow Map, g=0.5, r=0.5}
  n14 = FunctionInput() {
   bUsePreviewValueAsDefault=true
   description=This is a 0-1 determining the flow map direction
   inputName=Flow Direction
   inputType=FunctionInput_Scalar
   sortPriority=4
   previewValue=<0.25 0 0 1>
  }
  n15 = Frac(None=@n14)
  n16 = Multiply(A=Subtract(A=@n15) {constB=0.25}) {constB=4}
  n17 = OneMinus(None=@n15)
  n18 = Multiply(A=@n17) {constB=4}
  n19 = Multiply(A=Subtract(A=@n17) {constB=0.25}) {constB=4}
  n20 = Multiply(A=@n15) {constB=4}
  n21 = AppendVector(
   A=Multiply(A=Saturate(None=@n16), B=Saturate(None=@n18)) {constB=4, desc=XCoords}
   B=Multiply(A=Saturate(None=@n19), B=Saturate(None=@n20)) {constB=4, desc=YCoords}
  )
  n22 = TextureObject()
  n23 = FunctionInput(Preview=ComponentMask(None=TextureSample(Tex=@n22).RGB) {g=true, r=true}) {description=Single channel or 2 channels flow map, inputName=Grayscale Flow Map, inputType=FunctionInput_Vector2, sortPriority=3}
  n24 = FunctionInput() {
   bUsePreviewValueAsDefault=true
   description=Determine the intensity of the flow
   inputName=Flow Intensity
   inputType=FunctionInput_Scalar
   sortPriority=5
   previewValue=<0.4 0 0 1>
  }
  n25 = Multiply(A=NamedRerouteDeclaration(None=@n23) {name=Grayscale Flow Map, nodeColor=<rgba 0 0 0 1>}, B=@n24) {constB=4}
  NamedRerouteDeclaration(None=ConstantBiasScale(None=LinearInterpolate(A=@n13, B=@n21, Alpha=@n25)) {bias=-0.5, scale=2}) {name=Flow Map Coordinates, nodeColor=<rgba 0 0 0 1>}
 ]
}
