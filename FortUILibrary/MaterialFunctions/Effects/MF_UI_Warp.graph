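# UI material function that warps UVs around an adjustable center point, returning the warped UV and the mask used.
# Center-relative UVs go through a Rotator driven by intensity, then blend with the unrotated UVs through an inverted, saturated radial distance mask.
# Linearly blending coordinates, rather than varying the angle, gives the falloff warp.
# The radius input multiplies distance, so larger values shrink the affected area.
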MaterialFunction /FortUILibrary/MaterialFunctions/Effects/MF_UI_Warp
graph {
 properties {bExposeToLibrary=true, description=Allows you to warp your visual toward a center point., previewMaterialDomain=MD_UI}
 nodes [
  n1 = FunctionInput() {
   bUsePreviewValueAsDefault=true
   description="Defines the position of the warp center (Default = 0.5)"
   inputName=Warp Center
   inputType=FunctionInput_Vector2
   sortPriority=11
   previewValue=<0.5 0.5 0 0>
  }
  n2 = Subtract(
   A=FunctionInput(Preview=MaterialFunctionCall() {materialFunction=&GetUserInterfaceUV}.Normalized UV) {bUsePreviewValueAsDefault=true, description=Inputs your UV, inputName=UV, inputType=FunctionInput_Vector2}
   B=@n1
  ) {constB=0.5}
  n3 = FunctionInput() {
   bUsePreviewValueAsDefault=true
   description="Controls the strength of the warp effect (-1 to 1)"
   inputName=Warp Intensity
   inputType=FunctionInput_Scalar
   sortPriority=21
   previewValue=<0 0 0 0>
  }
  n4 = Rotator(Coordinate=@n2, Time=OneMinus(None=Add(A=@n3))) {centerX=0, centerY=0, speed=6.2832}
  n5 = FunctionInput() {
   bUsePreviewValueAsDefault=true
   description=Controls the radius of the warp effect
   inputName=Warp Radius
   inputType=FunctionInput_Scalar
   sortPriority=33
   previewValue=<1 0 0 0>
  }
  n6 = Length(None=Multiply(A=@n2, B=@n5)) {is=class MaterialExpressionLength}
  n7 = OneMinus(None=Saturate(None=Multiply(A=@n6) {constB=2}))
  FunctionOutput(None=Add(A=LinearInterpolate(A=@n2, B=@n4, Alpha=@n7), B=@n1) {constB=0.5}) {description=Outputs a warped version of your input UV, outputName=Warped UV}
  FunctionOutput(None=@n7) {description=Outputs the mask used to control the warping area, outputName=Warp Mask, sortPriority=1}
 ]
}
